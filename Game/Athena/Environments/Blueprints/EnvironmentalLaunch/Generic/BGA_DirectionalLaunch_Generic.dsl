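; Generic directional environmental launcher Blueprint that adds ambient audio, soft-deactivation cleanup and mesh orientation on top of its parent class.
; On BeginPlay it fades in the ambient audio.
; Soft deactivation calls the parent, then deactivates the idle loop effect and fades out the audio.
; The construction script optionally zeroes the temp mesh's world roll, keeping pitch and yaw, before running the parent script.
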(event EventBeginPlay
  (Audio|Components|Audio|FadeIn (Variables|Default|GetAmbientAudio) 0.25 1.0 0.0 "Linear" :!class /Script/Engine.AudioComponent))

(event EventSoftDeactivation
  ;[1 1 1] If Soft Deactivated turn off FX and Audio
  (|CallParentFunction)
  (Components|Activation|Deactivate (Variables|Default|GetLoopIdle))
  (Audio|Components|Audio|FadeOut (Variables|Default|GetAmbientAudio) 1.0 0.0 "Linear" :!class /Script/Engine.AudioComponent))

(fn ConstructionScript
  (bind (_returnvalue_roll _returnvalue_pitch _returnvalue_yaw) (Math|Rotator|BreakRotator (Transformation|GetWorldRotation (Variables|Default|GetMeshTemp))))
  (if (Variables|Default|GetFlattenRotationOnMesh)
    (Transformation|SetWorldRotation (Variables|Default|GetMeshTemp) (Math|Rotator|MakeRotator 0.0 _returnvalue_pitch _returnvalue_yaw) false false)
    (else
      _))
  (|CallParentFunction))
