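; Construction script of a UFO variant of the directional environmental launch Blueprint, which can level its mesh's orientation.
; When FlattenRotationOnMesh is set, it rewrites the world rotation of the MeshTemp component with zero roll, keeping its existing pitch and yaw.
; It then calls the parent construction script; only this script appears here, so any launch behavior lies outside it.
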(fn ConstructionScript
  (bind (_returnvalue_roll _returnvalue_pitch _returnvalue_yaw) (Math|Rotator|BreakRotator (Transformation|GetWorldRotation (Variables|Default|GetMeshTemp))))
  (if (Variables|Default|GetFlattenRotationOnMesh)
    (Transformation|SetWorldRotation (Variables|Default|GetMeshTemp) (Math|Rotator|MakeRotator 0.0 _returnvalue_pitch _returnvalue_yaw) false false)
    (else
      _))
  (|CallParentFunction))
