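; GA_DirectionalLaunched_Generic is a gameplay ability for the generic directional environmental launch pad in the Athena environment set.
; On activation it prints a debug string, "Generic Launch Pad Launched", then defers to the parent ability class, whose behavior is not shown here.
; The end-ability event is only a disabled placeholder, so this child adds no other logic.
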(event Ability|EventActivateAbility
  (Development|PrintString "Generic Launch Pad Launched")
  (|CallParentFunction))

(event Ability|EventOnEndAbility
  (params (bWasCancelled :type Boolean))
  (|CallParentFunction bWasCancelled) ; This node is disabled and will not be called.\nDrag off pins to build functionality.
)
